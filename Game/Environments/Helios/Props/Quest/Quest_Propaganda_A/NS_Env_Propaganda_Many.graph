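# Niagara system for the Helios propaganda quest prop, throwing a short burst of many paper mesh particles.
# Only GPU emitter Papers002 is enabled, spawning random mesh variants around offset cylinders; an older Papers variant is disabled.
# Curve-driven gravity lifts then drops sheets, combined with camera-relative attraction, curl noise, aerodynamic drag and plane-aligning collision.
# An active flag gates noise and drag, velocity gates aging, and distance shrinks and culls particles.
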NiagaraSystem /Game/Environments/Helios/Props/Quest/Quest_Propaganda_A/NS_Env_Propaganda_Many
dependencies {
 usedDataInterfaces [
  &NiagaraDataInterfaceCamera
  &NiagaraDataInterfaceCurve
  &NiagaraDataInterfaceVector2DCurve
  &NiagaraDataInterfaceCollisionQuery
 ]
 usedDynamicInputs []
 usedModules [&SystemState]
}
emitters {
 Papers {
  bEnabled=false
  simTarget=GPUComputeSim
  emitterUpdateScript [
   /SetVariables() {isSetParameters=true, module=SetVariables_#1}
   /SetVariables(
    Emitter.CameraForward:Vector3f=&ReturnCameraVector(Camera Property=::NewEnumerator1, Coordinate Space=::Simulation, CameraQuery=$Emitter.CameraRead)
    Emitter.CameraPosition:Vector3f=&ReturnCameraVector(Camera Property=::NewEnumerator0, Coordinate Space=::World, CameraQuery=$Emitter.CameraRead)
    Emitter.CameraRight:Vector3f=&ReturnCameraVector(Camera Property=::NewEnumerator3, Coordinate Space=::Simulation, CameraQuery=$Emitter.CameraRead)
    Emitter.CameraUp:Vector3f=&ReturnCameraVector(Camera Property=::NewEnumerator2, Coordinate Space=::Simulation, CameraQuery=$Emitter.CameraRead)
   ) {isSetParameters=true, module=SetVariables_#2}
   &EmitterState(
    Enable Distance Culling=true
    Life Cycle Mode=::NewEnumerator1
    Loop Behavior=::NewEnumerator1
    Loop Duration Mode=::NewEnumerator1
    Max Distance Response=::SleepAndClearParticles
    MaxDistance=8000
    Module.bUseMaxDistance=true
    Module.bUseMinDistance=false
    Reset Age On Awaken=true
    Scalability Mode=::NewEnumerator1
    UseLoopDelay=false
   )
   &SpawnRate(
    SpawnRate=&FloatFromCurve(FloatCurve=DataInterface() {propertyValues {Curve {keys [<at 0 0>,<at 0.1 1>,<at 0.2 0>]}}}, CurveIndex=$Emitter.Age, Scale Curve=3500)
    Use Spawn Probability=false
   )
  ]
  particleSpawnScript [
   /Niagara/Modules/Spawn/Initialization/InitializeParticle(
    Color=&MultiplyLinearColor(
     Linear Color=&RandomRangeLinearColor(Minimum=<rgba 1 0.4655 0.9726 1>, Maximum=<rgba 1 0.6927 0.9906 1>)
     Scale Factor=&MakeVector4_FromFloat(Float=1)
    )
    Lifetime=&RandomRangeFloat(Minimum=3, Maximum=5)
    Mass=&RandomRangeFloat(Minimum=3, Maximum=6)
    Mesh Scale=<0.05 0.05 0.05>
    Module.Write Color=true
    Module.Write Lifetime=true
    Module.Write Mass=true
    Module.Write Material Random=true
    Module.Write Position=false
    Module.Write Scale=true
    Module.Write SpriteRotation=true
    Module.Write SpriteSize=true
    Position=&SimulationPosition()
    Sprite Rotation=&RandomRangeFloat(Minimum=0, Maximum=360)
    Sprite Size=&Vector2DFromFloat(Value=&RandomRangeFloat(Minimum=8, Maximum=12))
   ) {enabled=false}
   /Niagara/Modules/Spawn/Initialization/V2/InitializeParticle(
    Color=&RandomRangeLinearColor(Minimum=<rgba 0.9123 0.9234 0.9896 1>, Maximum=<rgba 0.9167 0.9167 0.9167 1>)
    Lifetime=&RandomRangeFloat(Minimum=3, Maximum=5)
    Mesh Scale Max=<8 8 1>
    Mesh Scale Min=<6 6 0.6>
    Mesh Scale Mode=::NewEnumerator2
    Sprite Rotation Mode=::NewEnumerator2
    Sprite Size Mode=::NewEnumerator1
    Uniform Sprite Size Max=12
    Uniform Sprite Size Min=8
    UsePositionOffset=false
    Write Material Random=true
   ) {module=InitializeParticle001}
   /SetVariables(Particles.MeshIndex=&RandomRangeInt(Minimum=0, Maximum=3)) {isSetParameters=true, module=SetVariables_#3}
   /SetVariables(Particles.Terrain Offset=2) {isSetParameters=true, module=SetVariables_#4}
   &InitialMeshOrientation(
    Mesh Orientation Mode=::NewEnumerator1
    Rotation=&RandomRangeVector(Minimum=<0 0 0>, Maximum=<0 0 1>)
    Use Rotation Vector=true
   ) {module=InitialMeshOrientation001}
   /SetVariables(Particles.GridPosition:NiagaraPosition=$Particles.Position) {isSetParameters=true, module=SetVariables_#5}
   &ShapeLocation(
    Cylinder Height=1
    Cylinder Radius=150
    MaskSpawnGroup=false
    Non Uniform Scale=<1 1 0>
    Offset=&RandomRangeVector(Minimum=<-250 -250 0>, Maximum=<250 250 0>)
    Offset Mode=::NewEnumerator0
    Override Seed=false
    Shape Primitive=::NewEnumerator1
    Surface Only=true
    Surface Only Band Thickness=1
    Use Endcaps In Surface Only Mode=false
    UseOwnerScale=true
   )
   /SetVariables(Particles.Position=&Add_Vector(
     A=Linked() {linkedVariable {name=Particles.Position, type=NiagaraPosition}}
     B=&MakeVector(X=0, Y=0, Z=$Particles.Terrain Offset)
    )) {isSetParameters=true, module=SetVariables_#6}
   /SetVariables(Particles.New Vector 3f:Vector3f=<0 0 1>) {isSetParameters=true, module=SetVariables_#7}
   &AddVelocity()
   &CameraOffset(Camera Offset Amount=&RandomRangeFloat(Minimum=0, Maximum=1))
   &SubUVAnimation(End Frame=36, SubUV Animation Mode:ENiagara_SubUVLookupMode=::NewEnumerator2) {module=SubUVAnimation001}
   /SetVariables(Particles.SubImageIndex=&RandomRangeFloat(Minimum=0, Maximum=63)) {isSetParameters=true, module=SetVariables_#8}
   /SetVariables() {isSetParameters=true, module=SetVariables_#9}
   &InitialMeshOrientation(Use Rotation Vector=false) {enabled=false}
  ]
  particleUpdateScript [
   &ScaleColor(
    Scale Alpha=&FloatFromCurve(FloatCurve=DataInterface() {propertyValues {Curve {keys [<at 0.8 1>,<at 1 0>]}}}, Scale Curve=5)
    Scale RGB=&MakeVectorFromLinearColorRGB(Color=&MultiplyLinearColor(Linear Color=$Particles.Initial.Color, Scale Factor=&MakeVector4_FromFloat(Float=1)))
    ScaleA=true
    ScaleRGB=true
   )
   /Niagara/Modules/Update/Forces/PointAttractionForce(
    Attraction Radius=65
    AttractionStrength:NiagaraFloat=&Multiply_Float(
     A=&FloatFromCurve(FloatCurve=DataInterface() {propertyValues {Curve {keys [<at 0 1>,<at 0.5 0>]}}}, Scale Curve=-65)
     B=1
    )
    AttractorPosition:Vector3f=&Add_Vector(
     A=$Emitter.CameraPosition
     B=&Add_Vector(
      A=&Add_Vector(
       A=&Multiply_VectorByFloat(Vector=$Particles.New Vector 3f, Float=-130)
       B=&Multiply_VectorByFloat(Vector=$Emitter.CameraForward, Float=250)
      )
      B=&Multiply_VectorByFloat(Vector=$Emitter.CameraRight, Float=-50)
     )
    )
    Falloff Exponent=1
    Module.Kill Within Radius=false
    Module.Use Falloff=false
   )
   /SetVariables(Particles.Normalized Velocity:Vector3f=&NormalizeVector(Vector To Normalize=$Particles.Velocity)) {isSetParameters=true, module=SetVariables_#10}
   /SetVariables(Particles.Is Active:NiagaraBool=&Boolean_OrOperation(
     Bool1=$Particles.Is Active
     Bool2=&SetBoolByFloatComparison(
      A=&MakeFloatFromVector(Vector=Linked() {linkedVariable {name=Particles.Position, type=NiagaraPosition}}, Channel=::NewEnumerator2)
      B=&MakeFloatFromVector(Vector=Linked() {linkedVariable {name=Particles.GridPosition, type=NiagaraPosition}}, Channel=::NewEnumerator2)
      Comparison Type=::NewEnumerator6
     )
    )) {isSetParameters=true, module=SetVariables_#11}
   &CurlNoiseForce(
    Noise Frequency=25
    Noise Strength=&MakeCustomFloatFromBool(
     Bool=$Particles.Is Active
     True Float=&FloatFromCurve(FloatCurve=DataInterface() {propertyValues {Curve {keys [<at 0 1>,<at 1 0.1>]}}}, Scale Curve=1000)
     False Float=0
    )
    Pan Noise=false
   ) {module=CurlNoiseForce001}
   &Drag(
    Drag=&MakeCustomFloatFromBool(
     Bool=$Particles.Is Active
     True Float=&FloatFromCurve(
      FloatCurve=DataInterface() {propertyValues {Curve {keys [
          {interpMode=RCIM_Cubic, tangentMode=RCTM_User, time=0, value=0}
          {interpMode=RCIM_Cubic, tangentMode=RCTM_User, time=0.2, value=0.95}
          {interpMode=RCIM_Cubic, tangentMode=RCTM_User, time=0.5, value=0}
         ]}}}
      Scale Curve=0.25
     )
     False Float=0
    )
    Use Linear Drag=true
    Use Rotational Drag=false
   )
   &GravityForce(Gravity=&MakeVector(
     X=0
     Y=0
     Z=&FloatFromCurve(
      FloatCurve=DataInterface() {propertyValues {Curve {keys [
          {arriveTangent=9.0725, interpMode=RCIM_Cubic, leaveTangent=9.0725, tangentMode=RCTM_User, time=0, value=0}
          <at 0.0329 1>
          {arriveTangent=-2.5249, interpMode=RCIM_Cubic, leaveTangent=-2.5249, tangentMode=RCTM_User, time=0.1408, value=0.25}
          {arriveTangent=-0.0405, interpMode=RCIM_Cubic, leaveTangent=-0.0405, tangentMode=RCTM_User, time=0.25, value=-1}
         ]}}}
      Scale Curve=980
     )
    )) {module=GravityForce001}
   /SetVariables(Particles.Sprite Orientation:Vector3f=&Lerp_Vector(
     A=$Particles.New Vector 3f
     B=&ReturnCameraVector(Camera Vectors=::NewEnumerator1, Coordinate Space=::Simulation, CameraQuery=$Emitter.CameraRead)
     Alpha=&MakeCustomFloatFromBool(Bool=$Particles.Is Active, True Float=1, False Float=0)
    )) {isSetParameters=true, module=SetVariables_#12}
   &SubUVAnimation(End Frame=64, SubUV Animation Mode:ENiagara_SubUVLookupMode=::NewEnumerator1, SubUV Loop Count=4)
   &ParticleState(DeltaTime:NiagaraFloat=&MakeCustomFloatFromBool(
     Bool=&SetBoolByFloatComparison(A=&VectorLength(Vector=$Particles.Velocity), B=1)
     True Float=$Engine.DeltaTime
     False Float=0
    ))
   &ParticleState(Kill Particles When Lifetime Has Elapsed=false) {enabled=false, module=ParticleState001}
   &ScaleMeshSize(Scale Factor=&VectorFromFloat(Value=&Multiply_Float(
      A=&FloatFromCurve(
       FloatCurve=DataInterface() {propertyValues {Curve {keys [<at 0 0.25>,<at 100 1>,<at 6000 1>,<at 8000 0>]}}}
       CurveIndex=$Engine.Owner.LODDistance
       Scale Curve=0.5
      )
      B=&FloatFromCurve(FloatCurve=DataInterface() {propertyValues {Curve {keys [<at 0.8 1>,<at 1 0>]}}}, Scale Curve=0.25)
     )))
   &ScaleSpriteSize(
    Scale Sprite Size Mode=::NewEnumerator0
    Uniform Scale Factor=&Multiply_Float(
     A=&FloatFromCurve(
      FloatCurve=DataInterface() {propertyValues {Curve {keys [<at 0 1>,<at 500 1>,<at 6000 1>,<at 8000 0>]}}}
      CurveIndex=$Engine.Owner.LODDistance
      Scale Curve=2
     )
     B=&FloatFromCurve(FloatCurve=DataInterface() {propertyValues {Curve {keys [<at 0.8 1>,<at 1 0>]}}}, Scale Curve=0.25)
    )
   )
   /SetVariables(Particles.MeshOrientation=&LerpQuaternion(
     Quaternion A:Quat4f=$Particles.Initial.MeshOrientation
     Quaternion B:Quat4f=&FindQuatBetweenTwoVectors(From Vector=<1 0 0>, To Vector=$Particles.Normalized Velocity)
     Lerp Factor:NiagaraFloat=&ClampFloat(Float=&MakeFloatFromVector(Vector=$Particles.Normalized Velocity, Channel=::NewEnumerator0))
     Enable Spherical Linear Interpolation:NiagaraBool=true
    )) {enabled=false, isSetParameters=true, module=SetVariables_#13}
   &Color(Color=&MakeCustomLinearColorFromBool(Bool=$Particles.Is Close to Ground, True Color=<rgba 0 0.2682 1 1>, False Color=<rgba 0 1 0.2985 1>)) {enabled=false}
   &WindForce(
    Offset By Time:Vector3f=<0 0 -10>
    Scale=1
    Use Ground Mask=true
    Wind Speed=&MakeVector(
     X=&MakeCustomFloatFromBool(
      Bool=$Particles.Is Active
      True Float=&FloatFromCurve(FloatCurve=DataInterface() {propertyValues {Curve {keys [<at 0.25 1>,<at 0.5 0.25>]}}}, Scale Curve=100)
      False Float=0
     )
     Y=0
     Z=0
    )
   ) {module=WindForce001}
   &AerodynamicDrag(
    Aerodynamic Drag=0.1
    Pivot Offset Maximum:Vector3f=<1 1 1>
    Pivot Offset Minimum Length:Vector3f=<250 0 500>
    Rotation Strength:NiagaraFloat=&RandomRangeFloat(Minimum=5, Maximum=10)
    Use Linear Drag=true
    Use Rotational Drag=true
   )
   &Collision(
    Advanced Aging Rate=1
    Age Colliding Particles=false
    Collision Enabled=&SetBoolByFloatComparison(A=$Particles.Age, B=0.1)
    Enable Rest State=false
    Friction=1
    Friction During a Bounce=0.25
    GPU Collision Type=::NewEnumerator2
    Kill Particles Lodged Within Meshes=false
    KillOnCollision=false
    Module.EnableMaxCollisionCount=false
    Module.Randomize Collision Normal Vector=false
    Restitution=0.1
   )
   /Niagara/Modules/Collision/AlignParticlesWithCollisionPlane(Allow Alignment on X=false, Allow Alignment on Y=false, Calming Rate When Resting=25, Rotational Dampening Power=5)
   &SolveForcesAndVelocity(Clamp Velocity=true, Limit Acceleration=false)
   &AlignSpriteToMeshOrientation(Mesh Orientation Relative Sprite Alignment Vector=<0 1 1>, SetAlignment=true, SetFacing=true)
   &DynamicMaterialParameters(
    Index 0 Param 1=0.02
    Module.Param0WriteEnabled=true
    Module.Param1WriteEnabled=true
    Module.Param2WriteEnabled=true
    Module.Param3WriteEnabled=true
   ) {module=DynamicMaterialParameters001}
   &KillParticles(Kill Particles=&self:ScratchDynamicInput())
   /SetVariables(Particles.SpriteColor=&MultiplyLinearColor(Linear Color=$Particles.Color, Scale Factor=&MakeVector4_FromFloat(Float=5))) {isSetParameters=true, module=SetVariables_#14}
  ]
  rendererClasses [&NiagaraMeshRendererProperties]
  renderers [{rendererClass=&NiagaraMeshRendererProperties, rendererIndex=0}]
 }
 Papers002 {
  simTarget=GPUComputeSim
  emitterUpdateScript [
   /SetVariables() {isSetParameters=true, module=SetVariables_#1}
   /SetVariables(
    Emitter.CameraForward:Vector3f=&ReturnCameraVector(Camera Property=::NewEnumerator1, Coordinate Space=::Simulation, CameraQuery=$Emitter.CameraRead)
    Emitter.CameraPosition:Vector3f=&ReturnCameraVector(Camera Property=::NewEnumerator0, Coordinate Space=::World, CameraQuery=$Emitter.CameraRead)
    Emitter.CameraRight:Vector3f=&ReturnCameraVector(Camera Property=::NewEnumerator3, Coordinate Space=::Simulation, CameraQuery=$Emitter.CameraRead)
    Emitter.CameraUp:Vector3f=&ReturnCameraVector(Camera Property=::NewEnumerator2, Coordinate Space=::Simulation, CameraQuery=$Emitter.CameraRead)
   ) {isSetParameters=true, module=SetVariables_#2}
   &EmitterState(
    Enable Distance Culling=true
    Life Cycle Mode=::NewEnumerator1
    Loop Behavior=::NewEnumerator1
    Loop Duration Mode=::NewEnumerator1
    Max Distance Response=::SleepAndClearParticles
    MaxDistance=8000
    Module.bUseMaxDistance=true
    Module.bUseMinDistance=false
    Reset Age On Awaken=true
    Scalability Mode=::NewEnumerator1
    UseLoopDelay=false
   )
   &SpawnRate(
    SpawnRate=&FloatFromCurve(FloatCurve=DataInterface() {propertyValues {Curve {keys [<at 0 0>,<at 0.1 1>,<at 0.2 0>]}}}, CurveIndex=$Emitter.Age, Scale Curve=3500)
    Use Spawn Probability=false
   )
  ]
  particleSpawnScript [
   /Niagara/Modules/Spawn/Initialization/InitializeParticle(
    Color=&MultiplyLinearColor(
     Linear Color=&RandomRangeLinearColor(Minimum=<rgba 1 0.4655 0.9726 1>, Maximum=<rgba 1 0.6927 0.9906 1>)
     Scale Factor=&MakeVector4_FromFloat(Float=1)
    )
    Lifetime=&RandomRangeFloat(Minimum=3, Maximum=5)
    Mass=&RandomRangeFloat(Minimum=3, Maximum=6)
    Mesh Scale=<0.05 0.05 0.05>
    Module.Write Color=true
    Module.Write Lifetime=true
    Module.Write Mass=true
    Module.Write Material Random=true
    Module.Write Position=false
    Module.Write Scale=true
    Module.Write SpriteRotation=true
    Module.Write SpriteSize=true
    Position=&SimulationPosition()
    Sprite Rotation=&RandomRangeFloat(Minimum=0, Maximum=360)
    Sprite Size=&Vector2DFromFloat(Value=&RandomRangeFloat(Minimum=8, Maximum=12))
   ) {enabled=false}
   /Niagara/Modules/Spawn/Initialization/V2/InitializeParticle(
    Color=&RandomRangeLinearColor(Minimum=<rgba 0.9123 0.9234 0.9896 1>, Maximum=<rgba 0.7039 0.6702 0.6316 1>)
    Lifetime=&RandomRangeFloat(Minimum=3, Maximum=5)
    Mesh Scale Max=<10 10 1>
    Mesh Scale Min=<9 9 1>
    Mesh Scale Mode=::NewEnumerator2
    Sprite Rotation Mode=::NewEnumerator2
    Sprite Size Mode=::NewEnumerator1
    Uniform Sprite Size Max=12
    Uniform Sprite Size Min=8
    UsePositionOffset=false
    Write Material Random=true
   ) {module=InitializeParticle001}
   /SetVariables(Particles.MeshIndex=&RandomRangeInt(Minimum=0, Maximum=3)) {isSetParameters=true, module=SetVariables_#3}
   /SetVariables(Particles.Terrain Offset=2) {isSetParameters=true, module=SetVariables_#4}
   &InitialMeshOrientation(
    Mesh Orientation Mode=::NewEnumerator1
    Rotation=&RandomRangeVector(Minimum=<0 0 0>, Maximum=<0 0 1>)
    Use Rotation Vector=true
   ) {module=InitialMeshOrientation001}
   /SetVariables(Particles.GridPosition:NiagaraPosition=$Particles.Position) {isSetParameters=true, module=SetVariables_#5}
   &ShapeLocation(
    Cylinder Height=1
    Cylinder Radius=150
    MaskSpawnGroup=false
    Non Uniform Scale=<1 1 0>
    Offset=&RandomRangeVector(Minimum=<-250 -250 0>, Maximum=<250 250 0>)
    Offset Mode=::NewEnumerator0
    Override Seed=false
    Shape Primitive=::NewEnumerator1
    Surface Only=true
    Surface Only Band Thickness=1
    Use Endcaps In Surface Only Mode=false
    UseOwnerScale=true
   )
   /SetVariables(Particles.Position=&Add_Vector(
     A=Linked() {linkedVariable {name=Particles.Position, type=NiagaraPosition}}
     B=&MakeVector(X=0, Y=0, Z=$Particles.Terrain Offset)
    )) {isSetParameters=true, module=SetVariables_#6}
   /SetVariables(Particles.New Vector 3f:Vector3f=<0 0 1>) {isSetParameters=true, module=SetVariables_#7}
   &AddVelocity()
   &CameraOffset(Camera Offset Amount=&RandomRangeFloat(Minimum=0, Maximum=1))
   &SubUVAnimation(End Frame=36, SubUV Animation Mode:ENiagara_SubUVLookupMode=::NewEnumerator2) {module=SubUVAnimation001}
   /SetVariables(Particles.SubImageIndex=&RandomRangeFloat(Minimum=0, Maximum=63)) {isSetParameters=true, module=SetVariables_#8}
   /SetVariables() {isSetParameters=true, module=SetVariables_#9}
   &InitialMeshOrientation(Use Rotation Vector=false) {enabled=false}
  ]
  particleUpdateScript [
   &ScaleColor(
    Scale Alpha=&FloatFromCurve(FloatCurve=DataInterface() {propertyValues {Curve {keys [<at 0.8 1>,<at 1 0>]}}}, Scale Curve=5)
    Scale RGB=&MakeVectorFromLinearColorRGB(Color=&MultiplyLinearColor(Linear Color=$Particles.Initial.Color, Scale Factor=&MakeVector4_FromFloat(Float=1)))
    ScaleA=true
    ScaleRGB=true
   )
   &PointAttractionForce(
    Attraction Radius:NiagaraFloat=&RandomRangeFloat(Minimum=77, Maximum=55)
    AttractionStrength:NiagaraFloat=&Multiply_Float(
     A=&FloatFromCurve(FloatCurve=DataInterface() {propertyValues {Curve {keys [<at 0 1>,<at 0.5 1>]}}}, Scale Curve=2)
     B=1
    )
    AttractorPosition:Vector3f=&Add_Vector(
     A=$Emitter.CameraPosition
     B=&Add_Vector(
      A=&Add_Vector(
       A=&Multiply_VectorByFloat(Vector=$Particles.New Vector 3f, Float=-130)
       B=&Multiply_VectorByFloat(Vector=$Emitter.CameraForward, Float=250)
      )
      B=&Multiply_VectorByFloat(Vector=$Emitter.CameraRight, Float=-50)
     )
    )
    Module.Kill Within Radius=false
    Module.Use Falloff=false
   )
   /SetVariables(Particles.Normalized Velocity:Vector3f=&NormalizeVector(Vector To Normalize=$Particles.Velocity)) {isSetParameters=true, module=SetVariables_#10}
   /SetVariables(Particles.Is Active:NiagaraBool=&Boolean_OrOperation(
     Bool1=$Particles.Is Active
     Bool2=&SetBoolByFloatComparison(
      A=&MakeFloatFromVector(Vector=Linked() {linkedVariable {name=Particles.Position, type=NiagaraPosition}}, Channel=::NewEnumerator2)
      B=&MakeFloatFromVector(Vector=Linked() {linkedVariable {name=Particles.GridPosition, type=NiagaraPosition}}, Channel=::NewEnumerator2)
      Comparison Type=::NewEnumerator6
     )
    )) {isSetParameters=true, module=SetVariables_#11}
   &CurlNoiseForce(
    Noise Frequency=25
    Noise Strength=&MakeCustomFloatFromBool(
     Bool=$Particles.Is Active
     True Float=&FloatFromCurve(FloatCurve=DataInterface() {propertyValues {Curve {keys [<at 0 1>,<at 1 0.1>]}}}, Scale Curve=1000)
     False Float=0
    )
    Pan Noise=false
   ) {module=CurlNoiseForce001}
   &Drag(
    Drag=&MakeCustomFloatFromBool(
     Bool=$Particles.Is Active
     True Float=&FloatFromCurve(
      FloatCurve=DataInterface() {propertyValues {Curve {keys [
          {interpMode=RCIM_Cubic, tangentMode=RCTM_User, time=0, value=0.5}
          {interpMode=RCIM_Cubic, tangentMode=RCTM_User, time=0.2, value=0.95}
          {interpMode=RCIM_Cubic, tangentMode=RCTM_User, time=0.5, value=0.5}
         ]}}}
      Scale Curve=0.25
     )
     False Float=0
    )
    Use Linear Drag=true
    Use Rotational Drag=false
   )
   &GravityForce(Gravity=&MakeVector(
     X=0
     Y=0
     Z=&FloatFromCurve(
      FloatCurve=DataInterface() {propertyValues {Curve {keys [
          {arriveTangent=9.0725, interpMode=RCIM_Cubic, leaveTangent=9.0725, tangentMode=RCTM_User, time=0, value=0}
          <at 0.0329 1>
          {arriveTangent=-2.5249, interpMode=RCIM_Cubic, leaveTangent=-2.5249, tangentMode=RCTM_User, time=0.1408, value=0.25}
          {arriveTangent=-0.0405, interpMode=RCIM_Cubic, leaveTangent=-0.0405, tangentMode=RCTM_User, time=0.25, value=-1}
         ]}}}
      Scale Curve=980
     )
    )) {module=GravityForce001}
   /SetVariables(Particles.Sprite Orientation:Vector3f=&Lerp_Vector(
     A=$Particles.New Vector 3f
     B=&ReturnCameraVector(Camera Vectors=::NewEnumerator1, Coordinate Space=::Simulation, CameraQuery=$Emitter.CameraRead)
     Alpha=&MakeCustomFloatFromBool(Bool=$Particles.Is Active, True Float=1, False Float=0)
    )) {isSetParameters=true, module=SetVariables_#12}
   &SubUVAnimation(End Frame=64, SubUV Animation Mode:ENiagara_SubUVLookupMode=::NewEnumerator1, SubUV Loop Count=4)
   &ParticleState(DeltaTime:NiagaraFloat=&MakeCustomFloatFromBool(
     Bool=&SetBoolByFloatComparison(A=&VectorLength(Vector=$Particles.Velocity), B=1)
     True Float=$Engine.DeltaTime
     False Float=0
    ))
   &ParticleState(Kill Particles When Lifetime Has Elapsed=false) {enabled=false, module=ParticleState001}
   &ScaleMeshSize(Scale Factor=&VectorFromFloat(Value=&Multiply_Float(
      A=&FloatFromCurve(
       FloatCurve=DataInterface() {propertyValues {Curve {keys [<at 0 0.25>, <at 100 1>, {arriveTangent=-0.0001, interpMode=RCIM_Cubic, leaveTangent=-0.0001, time=6000, value=1}, <at 8000 0 RCIM_Cubic>]}}}
       CurveIndex=$Engine.Owner.LODDistance
       Scale Curve=0.5
      )
      B=&FloatFromCurve(
       FloatCurve=DataInterface() {propertyValues {Curve {keys [<at 0.7 1 RCIM_Cubic>, {arriveTangent=-9.4406, interpMode=RCIM_Cubic, leaveTangent=-9.4406, tangentMode=RCTM_User, time=1, value=0}]}}}
       Scale Curve=0.25
      )
     )))
   &ScaleSpriteSize(
    Scale Sprite Size Mode=::NewEnumerator0
    Uniform Scale Factor=&Multiply_Float(
     A=&FloatFromCurve(
      FloatCurve=DataInterface() {propertyValues {Curve {keys [<at 0 1>, <at 500 1>, {arriveTangent=-0.0001, interpMode=RCIM_Cubic, leaveTangent=-0.0001, time=6000, value=1}, <at 8000 0 RCIM_Cubic>]}}}
      CurveIndex=$Engine.Owner.LODDistance
      Scale Curve=2
     )
     B=&FloatFromCurve(FloatCurve=DataInterface() {propertyValues {Curve {keys [<at 0.9 1 RCIM_Cubic>,<at 1 0 RCIM_Cubic>]}}}, Scale Curve=0.25)
    )
   )
   /SetVariables(Particles.MeshOrientation=&LerpQuaternion(
     Quaternion A:Quat4f=$Particles.Initial.MeshOrientation
     Quaternion B:Quat4f=&FindQuatBetweenTwoVectors(From Vector=<1 0 0>, To Vector=$Particles.Normalized Velocity)
     Lerp Factor:NiagaraFloat=&ClampFloat(Float=&MakeFloatFromVector(Vector=$Particles.Normalized Velocity, Channel=::NewEnumerator0))
     Enable Spherical Linear Interpolation:NiagaraBool=true
    )) {enabled=false, isSetParameters=true, module=SetVariables_#13}
   &Color(Color=&MakeCustomLinearColorFromBool(Bool=$Particles.Is Close to Ground, True Color=<rgba 0 0.2682 1 1>, False Color=<rgba 0 1 0.2985 1>)) {enabled=false}
   &WindForce(
    Offset By Time:Vector3f=<0 0 -10>
    Scale=1
    Use Ground Mask=true
    Wind Speed=&MakeVector(
     X=&MakeCustomFloatFromBool(
      Bool=$Particles.Is Active
      True Float=&FloatFromCurve(FloatCurve=DataInterface() {propertyValues {Curve {keys [<at 0.1 1>,<at 0.2 0.25>]}}}, Scale Curve=100)
      False Float=0
     )
     Y=0
     Z=0
    )
   ) {enabled=false, module=WindForce001}
   &AerodynamicDrag(
    Aerodynamic Drag=0.3
    Lift Contribution:NiagaraFloat=&RandomRangeFloat(Minimum=1, Maximum=0.2)
    Pivot Offset Maximum:Vector3f=<1 1 1>
    Pivot Offset Minimum Length:Vector3f=<250 0 500>
    Rotation Strength:NiagaraFloat=&RandomRangeFloat(Minimum=1, Maximum=5)
    Use Linear Drag=true
    Use Rotational Drag=true
   )
   &Collision(
    Advanced Aging Rate=1
    Age Colliding Particles=false
    Collision Enabled=&SetBoolByFloatComparison(A=$Particles.Age, B=0.1)
    Enable Rest State=false
    Friction=1
    Friction During a Bounce=0.25
    GPU Collision Type=::NewEnumerator2
    Kill Particles Lodged Within Meshes=false
    KillOnCollision=false
    Module.EnableMaxCollisionCount=false
    Module.Randomize Collision Normal Vector=false
    Restitution=0.1
   )
   &AlignParticlesWithCollisionPlane(Allow Alignment on X=false, Allow Alignment on Y=false, Calming Rate When Resting=25, Rotational Dampening Power=5)
   &SolveForcesAndVelocity(Clamp Velocity=true, Limit Acceleration=false)
   &AlignSpriteToMeshOrientation(Mesh Orientation Relative Sprite Alignment Vector=<0 1 1>, SetAlignment=true, SetFacing=true)
   &DynamicMaterialParameters(
    Index 0 Param 1=0.02
    Index 0 Param 4=&FloatFromCurve(FloatCurve=DataInterface() {propertyValues {Curve {keys [
         {arriveTangent=4.1785, interpMode=RCIM_Cubic, leaveTangent=4.1785, tangentMode=RCTM_User, time=0, value=0}
         {arriveTangent=0.1622, interpMode=RCIM_Cubic, leaveTangent=0.1622, tangentMode=RCTM_User, time=0.1, value=1}
         {arriveTangent=0.0224, interpMode=RCIM_Cubic, leaveTangent=0.0224, tangentMode=RCTM_User, time=0.5, value=0.95}
         <at 0.8 0>
        ]}}})
    Module.Param0WriteEnabled=true
    Module.Param1WriteEnabled=true
    Module.Param2WriteEnabled=true
    Module.Param3WriteEnabled=true
   ) {module=DynamicMaterialParameters001}
   &DynamicMaterialParameters(
    Index 0 Param 1=0.02
    Index 0 Param 4=&FloatFromCurve(
     FloatCurve=DataInterface() {propertyValues {Curve {keys [
         {arriveTangent=4.1785, interpMode=RCIM_Cubic, leaveTangent=4.1785, tangentMode=RCTM_User, time=0, value=0}
         {arriveTangent=1.1179, interpMode=RCIM_Cubic, leaveTangent=1.1179, tangentMode=RCTM_User, time=0.1, value=1}
         {arriveTangent=0.0224, interpMode=RCIM_Cubic, leaveTangent=0.0224, tangentMode=RCTM_User, time=1, value=1}
        ]}}}
     CurveIndex=&VectorLength(Vector=$Particles.Velocity)
    )
    Module.Param0WriteEnabled=true
    Module.Param1WriteEnabled=true
    Module.Param2WriteEnabled=true
    Module.Param3WriteEnabled=true
   ) {enabled=false}
   &KillParticles(Kill Particles=&self:ScratchDynamicInput())
   /SetVariables(Particles.SpriteColor=&MultiplyLinearColor(Linear Color=$Particles.Color, Scale Factor=&MakeVector4_FromFloat(Float=5))) {isSetParameters=true, module=SetVariables_#14}
  ]
  rendererClasses [&NiagaraMeshRendererProperties]
  renderers [{rendererClass=&NiagaraMeshRendererProperties, rendererIndex=0}]
 }
}
summary {
 systemName=NS_Env_Propaganda_Many
 emitters [
  {bEnabled=false, emitterName=Papers, simTarget=GPUComputeSim, rendererClasses [&NiagaraMeshRendererProperties]}
  {emitterName=Papers002, simTarget=GPUComputeSim, rendererClasses [&NiagaraMeshRendererProperties]}
 ]
 userVariables [
  User.Landscape {
   type=NiagaraDataInterfaceLandscape
   description=""
   defaultValue {
    struct=NiagaraExt_VariableValue_DataInterface
    value {
     dataInterface=&self:NiagaraDataInterfaceLandscape_0
     dataInterfaceClass=&NiagaraDataInterfaceLandscape
    }
   }
  }
  User.MaxSize {type=NiagaraFloat, defaultValue=1, description=""}
  User.MinSize {type=NiagaraFloat, defaultValue=0.5, description=""}
  User.RadiusSpread {type=NiagaraFloat, defaultValue=100, description=""}
  User.SpawnAmount {type=NiagaraFloat, defaultValue=15, description=""}
  User.TerrainOffset {type=NiagaraFloat, defaultValue=5, description=""}
 ]
}
